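# Physics asset for the camera and strap on the Photographer Holiday female medium player body, set up for secondary dangling motion.
# Capsules on dyn_camera and two strap bones form a chain hung from spine_03, with tight swing and twist limits at each joint.
# Capsules on the pelvis, spine and left thigh and calf give the camera body shapes to collide with, linked by constraints with free angular motion.
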PhysicsAsset /BRCosmetics/Characters/Player/Female/Medium/Bodies/F_MED_Photographer_Holiday/Meshes/F_MED_Photographer_Holiday_Camera_Physics
bodies {
 calf_l [{
   shapeType=Capsule
   length=8.5648
   radius=6.8696
   shapeName=calf_l_capsule
   center=<-8.321 -2.8296 -0.1798>
   rotation=<rot 0 -100.8419 -89.1998>
  }]
 dyn_camera [{shapeType=Capsule, length=4, radius=3, shapeName=dyn_camera_capsule, center=<2 0 0>, rotation=<rot 90 0 0>}]
 dyn_cameraStrap_02 [{shapeType=Capsule, length=4, radius=2.5, shapeName=dyn_cameraStrap_02_capsule, center=<2 0 0>, rotation=<rot 90 -109.5974 -109.5976>}]
 dyn_cameraStrap_03 [{shapeType=Capsule, length=4, radius=3, shapeName=dyn_cameraStrap_03_capsule, center=<2 0 0>, rotation=<rot 90 -102.5972 -102.5973>}]
 pelvis [{shapeType=Capsule, length=8, radius=10, shapeName=pelvis_capsule}]
 spine_02 [{shapeType=Capsule, length=8, radius=7, shapeName=spine_03_capsule, center=<0 0 1.3453>}]
 spine_03 [{shapeType=Capsule, length=1, radius=7, shapeName=spine_03_capsule, center=<0 0 1.2732>}]
 thigh_l [{shapeType=Capsule, length=20, radius=8.5, shapeName=thigh_l_capsule, center=<-12.1329 0.4021 0>, rotation=<rot 90 180 -174.9999>}]
}
constraints [
 {bone1Name=dyn_camera, bone2Name=dyn_cameraStrap_03, swing1LimitDegrees=15, swing2LimitDegrees=15, twistLimitDegrees=5}
 {bone1Name=dyn_cameraStrap_03, bone2Name=dyn_cameraStrap_02, swing1LimitDegrees=20, swing2LimitDegrees=20, twistLimitDegrees=5}
 {bone1Name=dyn_cameraStrap_02, bone2Name=spine_03, swing1LimitDegrees=10, swing2LimitDegrees=25, twistLimitDegrees=3}
 {
  bone1Name=spine_03
  bone2Name=spine_02
  swing1LimitDegrees=45
  swing1Motion=Free
  swing2LimitDegrees=45
  swing2Motion=Free
  twistLimitDegrees=45
  twistMotion=Free
 }
 {
  bone1Name=thigh_l
  bone2Name=pelvis
  swing1LimitDegrees=45
  swing1Motion=Free
  swing2LimitDegrees=45
  swing2Motion=Free
  twistLimitDegrees=45
  twistMotion=Free
 }
 {
  bone1Name=calf_l
  bone2Name=pelvis
  swing1LimitDegrees=45
  swing1Motion=Free
  swing2LimitDegrees=45
  swing2Motion=Free
  twistLimitDegrees=45
  twistMotion=Free
 }
 {
  bone1Name=spine_02
  bone2Name=pelvis
  swing1LimitDegrees=45
  swing1Motion=Free
  swing2LimitDegrees=45
  swing2Motion=Free
  twistLimitDegrees=45
  twistMotion=Free
 }
]
