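# Physics asset for the scarf on the female medium Photographer Holiday character, giving its dynamic bones bodies for secondary motion.
# Capsules form a three-bone front chain and a two-bone back chain, each rooted on spine_05 by constraints with small twist limit values.
# Larger capsules on both clavicles and the upper spine supply torso volume, with the clavicles rotationally locked to spine_05 and spine_04 left unconstrained.
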PhysicsAsset /BRCosmetics/Characters/Player/Female/Medium/Bodies/F_MED_Photographer_Holiday/Meshes/F_MED_Photographer_Holiday_Scarf_FE_Physics
bodies {
 clavicle_l [{shapeType=Capsule, length=4.5, radius=4.75, shapeName=clavicle_l_capsule, center=<6.807 0 0>, rotation=<rot -80 179.9994 -179.9997>}]
 clavicle_r [{shapeType=Capsule, length=4.5, radius=4.75, shapeName=clavicle_r_capsule, center=<-6.8573 0.8148 -0.285>, rotation=<rot 80 -0.0001 -0.0003>}]
 dyn_scarf_bck_01 [{
   shapeType=Capsule
   length=5.4034
   radius=1.9071
   shapeName=dyn_scarf_bck_01_capsule
   center=<1.5 -1 1>
   rotation=<rot 12.563 -89.9999 14.9867>
  }]
 dyn_scarf_bck_02 [{
   shapeType=Capsule
   length=4.5655
   radius=2.5738
   shapeName=dyn_scarf_bck_02_capsule
   center=<1 -1.5 0.5>
   rotation=<rot 8.1201 178.6862 -9.2253>
  }]
 dyn_scarf_fr_01 [{shapeType=Capsule, length=3.6764, radius=3.9179, shapeName=dyn_scarf_fr_01_capsule, center=<3 0 3>, rotation=<rot 0 173.7235 -70.7199>}]
 dyn_scarf_fr_02 [{shapeType=Capsule, length=4, radius=5, shapeName=dyn_scarf_fr_02_capsule, center=<1.5438 -3.2 1.5>, rotation=<rot 26.0162 -90 6.0852>}]
 dyn_scarf_fr_03 [{shapeType=Capsule, length=4, radius=5, shapeName=dyn_scarf_fr_03_capsule, center=<1 -3.5 0>, rotation=<rot 14.5595 -91.2286 6.4515>}]
 spine_04 [{shapeType=Capsule, length=5.122, radius=7.0025, shapeName=spine_04_capsule, center=<1.0613 3.5388 -2.3279>, rotation=<rot 0 -0.0001 30>}]
 spine_05 [{shapeType=Capsule, length=10.5, radius=7.5, shapeName=spine_05_capsule, center=<2.4231 -0.1411 0>}]
}
constraints [
 {
  bone1Name=clavicle_l
  bone2Name=spine_05
  swing1LimitDegrees=45
  swing1Motion=Locked
  swing2LimitDegrees=45
  swing2Motion=Locked
  twistLimitDegrees=45
  twistMotion=Locked
 }
 {
  bone1Name=clavicle_r
  bone2Name=spine_05
  swing1LimitDegrees=45
  swing1Motion=Locked
  swing2LimitDegrees=45
  swing2Motion=Locked
  twistLimitDegrees=45
  twistMotion=Locked
 }
 {bone1Name=dyn_scarf_fr_03, bone2Name=dyn_scarf_fr_02, swing1LimitDegrees=35, swing2LimitDegrees=15, twistLimitDegrees=5}
 {bone1Name=dyn_scarf_fr_01, bone2Name=spine_05, swing1LimitDegrees=35, swing2LimitDegrees=35, twistLimitDegrees=5}
 {bone1Name=dyn_scarf_fr_02, bone2Name=dyn_scarf_fr_01, swing1LimitDegrees=35, swing2LimitDegrees=15, twistLimitDegrees=5}
 {bone1Name=dyn_scarf_bck_01, bone2Name=spine_05, swing1LimitDegrees=35, swing2LimitDegrees=15, twistLimitDegrees=3}
 {bone1Name=dyn_scarf_bck_02, bone2Name=dyn_scarf_bck_01, swing1LimitDegrees=35, swing2LimitDegrees=15, twistLimitDegrees=3}
]
