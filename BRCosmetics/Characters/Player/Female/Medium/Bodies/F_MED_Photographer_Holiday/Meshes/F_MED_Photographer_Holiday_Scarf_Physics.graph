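# Physics asset for the scarf on the Photographer Holiday medium female player cosmetic body, providing capsule bodies for its dynamic scarf bones.
# A two-bone back chain and a three-bone front chain are constrained link by link up to spine_05, with twist locked on the back chain.
# Upper spine and clavicle capsules supply torso shapes, with both clavicles rotationally locked to spine_05 and spine_04 left without a constraint.
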PhysicsAsset /BRCosmetics/Characters/Player/Female/Medium/Bodies/F_MED_Photographer_Holiday/Meshes/F_MED_Photographer_Holiday_Scarf_Physics
bodies {
 clavicle_l [{shapeType=Capsule, length=4.5, radius=4.75, shapeName=clavicle_l_capsule, center=<6.807 0 0>, rotation=<rot -80 179.9994 -179.9997>}]
 clavicle_r [{shapeType=Capsule, length=4.5, radius=4.75, shapeName=clavicle_r_capsule, center=<-6.8573 0.8148 -0.285>, rotation=<rot 80 -0.0001 -0.0003>}]
 dyn_scarf_bck_01 [{
   shapeType=Capsule
   length=5.221
   radius=2.0895
   shapeName=dyn_scarf_bck_01_capsule
   center=<1.804 -0.8623 1.6782>
   rotation=<rot 12.563 -89.9999 14.9867>
  }]
 dyn_scarf_bck_02 [{
   shapeType=Capsule
   length=4.3831
   radius=2.7562
   shapeName=dyn_scarf_bck_02_capsule
   center=<0.8098 -1.6739 0.5344>
   rotation=<rot 8.1201 178.6862 -9.2253>
  }]
 dyn_scarf_fr_01 [{shapeType=Capsule, length=4.4264, radius=3.1679, shapeName=dyn_scarf_fr_01_capsule, center=<2 0 2.5>, rotation=<rot 0.0001 0 75>}]
 dyn_scarf_fr_02 [{shapeType=Capsule, length=4.9658, radius=3.4187, shapeName=dyn_scarf_fr_02_capsule, center=<2 -2 1>, rotation=<rot 0.0001 0 35>}]
 dyn_scarf_fr_03 [{shapeType=Capsule, length=5.2372, radius=3.6983, shapeName=dyn_scarf_fr_03_capsule, center=<2 -2 0>, rotation=<rot 0.0001 -0.0001 30>}]
 spine_04 [{shapeType=Capsule, length=7.872, radius=8.0025, shapeName=spine_04_capsule, center=<1.0613 1.9726 -5.0405>, rotation=<rot 0 -0.0001 30>}]
 spine_05 [{shapeType=Capsule, length=11.75, radius=8, shapeName=spine_05_capsule, center=<2.4231 -0.075 0>}]
}
constraints [
 {
  bone1Name=clavicle_l
  bone2Name=spine_05
  swing1LimitDegrees=45
  swing1Motion=Locked
  swing2LimitDegrees=45
  swing2Motion=Locked
  twistLimitDegrees=45
  twistMotion=Locked
 }
 {
  bone1Name=clavicle_r
  bone2Name=spine_05
  swing1LimitDegrees=45
  swing1Motion=Locked
  swing2LimitDegrees=45
  swing2Motion=Locked
  twistLimitDegrees=45
  twistMotion=Locked
 }
 {
  bone1Name=dyn_scarf_bck_02
  bone2Name=dyn_scarf_bck_01
  swing1LimitDegrees=45
  swing2LimitDegrees=10
  twistLimitDegrees=45
  twistMotion=Locked
 }
 {bone1Name=dyn_scarf_bck_01, bone2Name=spine_05, swing1LimitDegrees=45, swing2LimitDegrees=10, twistLimitDegrees=10, twistMotion=Locked}
 {bone1Name=dyn_scarf_fr_01, bone2Name=spine_05, swing1LimitDegrees=20, swing2LimitDegrees=20, twistLimitDegrees=5}
 {bone1Name=dyn_scarf_fr_02, bone2Name=dyn_scarf_fr_01, swing1LimitDegrees=20, swing2LimitDegrees=20, twistLimitDegrees=5}
 {bone1Name=dyn_scarf_fr_03, bone2Name=dyn_scarf_fr_02, swing1LimitDegrees=20, swing2LimitDegrees=20, twistLimitDegrees=5}
]
