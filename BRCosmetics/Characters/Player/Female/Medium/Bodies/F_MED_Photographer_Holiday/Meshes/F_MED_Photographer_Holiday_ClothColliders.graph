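# Physics asset that supplies cloth collision shapes for the F_MED_Photographer_Holiday medium female player character body.
# Single capsules on spine_01, thigh_l and thigh_r approximate the lower torso and upper legs, while the pelvis body has no shapes.
# It defines no constraints, and both thigh capsules share the name thigh_r_capsule despite differing slightly in size and placement.
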PhysicsAsset /BRCosmetics/Characters/Player/Female/Medium/Bodies/F_MED_Photographer_Holiday/Meshes/F_MED_Photographer_Holiday_ClothColliders
bodies {
 pelvis []
 spine_01 [{shapeType=Capsule, length=9.4153, radius=6.505, shapeName=spine_01_capsule, center=<-2.4928 -0.9588 0>, rotation=<rot 0 -90 0>}]
 thigh_l [{
   shapeType=Capsule
   length=10.6064
   radius=7.6436
   shapeName=thigh_r_capsule
   center=<-8.4952 -2.1871 2.4896>
   rotation=<rot 85 179.9996 179.9999>
  }]
 thigh_r [{
   shapeType=Capsule
   length=10.0648
   radius=8.1852
   shapeName=thigh_r_capsule
   center=<8.4952 1.5529 -2.4896>
   rotation=<rot 85 179.9996 179.9999>
  }]
}
constraints []
