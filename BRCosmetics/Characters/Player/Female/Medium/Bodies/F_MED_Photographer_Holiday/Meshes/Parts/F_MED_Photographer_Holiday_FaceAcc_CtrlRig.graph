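# Control Rig for the female Photographer Holiday face accessory, applying pose-driven corrective offsets to dynamic head bones.
# Spherical pose readers on the head drive weighted Modify Transforms for look-up and a second head direction.
# Right upper-arm readers are multiplied by the look-up alpha to add corrections for two-handed and rifle aim-up poses, the latter smoothed by AlphaInterp.
# A hair-bone set-transform chain and two debug-drawing readers are left unconnected to execution.
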ControlRigBlueprint /BRCosmetics/Characters/Player/Female/Medium/Bodies/F_MED_Photographer_Holiday/Meshes/Parts/F_MED_Photographer_Holiday_FaceAcc_CtrlRig
graphs [
 RigVMFunctionLibrary {links [], nodes []}
 RigVMModel {nodes [
   RigUnit_BeginExecution = &RigUnit_BeginExecution()
   SequenceAggregate = &RigVMFunction_Sequence(ExecuteContext=@RigUnit_BeginExecution.ExecutePin)
   n1 = &RigUnit_SphericalPoseReader(
    ActiveRegionScaleFactors=(PositiveWidth=1,NegativeWidth=0.4,PositiveHeight=1.5,NegativeHeight=1.5)
    ActiveRegionSize=0.1
    Debug=(bDrawDebug=false,DebugScale=25,DebugSegments=20,DebugThickness=0.25)
    DriverAxis=(X=0,Y=1,Z=0)
    DriverItem=(Type=Bone,Name="head")
    FalloffRegionScaleFactors=(PositiveWidth=1,NegativeWidth=0.5,PositiveHeight=1.5,NegativeHeight=1.5)
    FalloffSize=0.45
    FlipWidthScaling=false
    OptionalParentItem=(Type=Bone,Name="spine_05")
    RotationOffset=(X=90,Y=0,Z=-45)
    ExecutePin=@SequenceAggregate.A
   )
   > RigVMVariableNode(Value=@n1.OutputParam) {variable=head_SPR_lookup_alpha}
   > &RigUnit_ModifyTransforms(
    ItemToModify="((Item=(Type=Bone,Name=\"dyn_headBack_Base\"),Transform=(Rotation=(X=0,Y=0,Z=0,W=1),Translation=(X=-4,Y=0,Z=1),Scale3D=(X=1,Y=1,Z=1))),(Item=(Type=Bone,Name=\"dyn_headBack_01_r\"),Transform=(Rotation=(X=0,Y=0,Z=0,W=1),Translation=(X=0,Y=0,Z=0),Scale3D=(X=1,Y=1 ...(truncated)"
    ItemToModify.0.Transform.Rotation=&RigVMFunction_MathQuaternionFromEuler(Euler=(X=0,Y=0,Z=-25)).Result
    ItemToModify.1.Transform.Rotation=&RigVMFunction_MathQuaternionFromEuler(Euler=(X=0,Y=-5,Z=-5)).Result
    ItemToModify.2.Transform.Rotation=&RigVMFunction_MathQuaternionFromEuler(Euler=(X=0,Y=-5,Z=-5)).Result
    Weight=RigVMVariableNode() {variable=head_SPR_lookup_alpha}.Value
   )
   > n2 = &RigUnit_SphericalPoseReader(
    ActiveRegionScaleFactors=(PositiveWidth=2,NegativeWidth=0.5,PositiveHeight=2,NegativeHeight=2)
    ActiveRegionSize=0.075
    Debug=(bDrawDebug=false,DebugScale=25,DebugSegments=20,DebugThickness=0.25)
    DriverAxis=(X=0,Y=1,Z=0)
    DriverItem=(Type=Bone,Name="head")
    FalloffRegionScaleFactors=(PositiveWidth=1,NegativeWidth=0.4,PositiveHeight=1,NegativeHeight=1)
    FalloffSize=0.3
    FlipHeightScaling=false
    FlipWidthScaling=true
    OptionalParentItem=(Type=Bone,Name="spine_05")
    RotationOffset=(X=90,Y=0,Z=25)
   )
   > &RigUnit_ModifyTransforms(
    ItemToModify="((Item=(Type=Bone,Name=\"dyn_headBack_Base\"),Transform=(Rotation=(X=0,Y=0,Z=0,W=1),Translation=(X=0,Y=0,Z=0),Scale3D=(X=1,Y=1,Z=1))),(Item=(Type=Bone,Name=\"dyn_head_01_Base_r\"),Transform=(Rotation=(X=0,Y=0,Z=0,W=1),Translation=(X=0,Y=0,Z=0),Scale3D=(X=1,Y=1 ...(truncated)"
    ItemToModify.2.Transform.Rotation=&RigVMFunction_MathQuaternionFromEuler(Euler=(X=0,Y=-20,Z=0)).Result
    ItemToModify.1.Transform.Rotation=&RigVMFunction_MathQuaternionFromEuler(Euler=(X=0,Y=20,Z=0)).Result
    ItemToModify.0.Transform.Rotation=&RigVMFunction_MathQuaternionFromEuler(Euler=(X=0,Y=0,Z=15)).Result
    Weight=@n2.OutputParam
   )
   > &RigUnit_ModifyTransforms(
    ItemToModify="((Item=(Type=Bone,Name=\"dyn_headFront_01_Base_l\"),Transform=(Rotation=(X=0,Y=0,Z=0,W=1),Translation=(X=0,Y=0,Z=0),Scale3D=(X=1,Y=1,Z=1))),(Item=(Type=Bone,Name=\"dyn_headFront_01_Base_r\"),Transform=(Rotation=(X=0,Y=0,Z=0,W=1),Translation=(X=0,Y=0,Z=0),Scale ...(truncated)"
    ItemToModify.0.Transform.Rotation=&RigVMFunction_MathQuaternionFromEuler(Euler=(X=0,Y=-20,Z=0)).Result
    ItemToModify.1.Transform.Rotation=&RigVMFunction_MathQuaternionFromEuler(Euler=(X=0,Y=20,Z=0)).Result
    Weight=@n2.OutputParam
   )
   n3 = &RigUnit_SphericalPoseReader(
    ActiveRegionScaleFactors=(PositiveWidth=0.6,NegativeWidth=0.6,PositiveHeight=0.8,NegativeHeight=0.8)
    ActiveRegionSize=0.1
    Debug=(bDrawDebug=false,bDraw2D=false,bDrawLocalAxes=true,DebugScale=25,DebugSegments=20,DebugThickness=0.25)
    DriverAxis=(X=-1,Y=0,Z=0)
    DriverItem=(Type=Bone,Name="upperarm_r")
    FalloffRegionScaleFactors=(PositiveWidth=0.6,NegativeWidth=0.6,PositiveHeight=1,NegativeHeight=1)
    FalloffSize=0.2
    FlipHeightScaling=false
    FlipWidthScaling=true
    OptionalParentItem=(Type=Bone,Name="spine_04")
    RotationOffset=(X=-110,Y=-35,Z=50)
    ExecutePin=@SequenceAggregate.B
   )
   > RigVMVariableNode(Value=@n3.OutputParam) {variable=uppearm_r_SPR_aim_up_two_hands_alpha}
   > RigVMVariableNode(Value=&RigVMFunction_MathFloatMul(B=RigVMVariableNode() {variable=uppearm_r_SPR_aim_up_two_hands_alpha}.Value, A=RigVMVariableNode() {variable=head_SPR_lookup_alpha}.Value) {template="Multiply::Execute(in A,in B,out Result)"}.Result) {variable=look_up_and_left_two_hands_multiplied_alpha}
   > &RigUnit_ModifyTransforms(
    ItemToModify="((Item=(Type=Bone,Name=\"dyn_headBack_Base\"),Transform=(Rotation=(X=0,Y=0,Z=0,W=1),Translation=(X=0,Y=0,Z=1),Scale3D=(X=1,Y=1,Z=1))),(Item=(Type=Bone,Name=\"dyn_headBack_01_r\"),Transform=(Rotation=(X=0,Y=0,Z=0,W=1),Translation=(X=-2,Y=-2,Z=0),Scale3D=(X=1,Y= ...(truncated)"
    ItemToModify.0.Transform.Rotation=&RigVMFunction_MathQuaternionFromEuler(Euler=(X=0,Y=10,Z=-10)).Result
    ItemToModify.2.Transform.Rotation=&RigVMFunction_MathQuaternionFromEuler(Euler=(X=-10,Y=-15,Z=-10)).Result
    ItemToModify.1.Transform.Rotation=&RigVMFunction_MathQuaternionFromEuler(Euler=(X=0,Y=-10,Z=0)).Result
    ItemToModify.3.Transform.Rotation=&RigVMFunction_MathQuaternionFromEuler(Euler=(X=0,Y=0,Z=-15)).Result
    Weight=RigVMVariableNode() {variable=look_up_and_left_two_hands_multiplied_alpha}.Value
   )
   > n4 = &RigUnit_SphericalPoseReader(
    ActiveRegionScaleFactors=(PositiveWidth=1,NegativeWidth=1,PositiveHeight=1,NegativeHeight=1)
    ActiveRegionSize=0.075
    Debug=(bDrawDebug=false,DebugScale=25,DebugSegments=20,DebugThickness=0.25)
    DriverAxis=(X=-1,Y=0,Z=0)
    DriverItem=(Type=Bone,Name="upperarm_r")
    FalloffRegionScaleFactors=(PositiveWidth=1,NegativeWidth=1,PositiveHeight=1,NegativeHeight=1)
    FalloffSize=0.2
    OptionalParentItem=(Type=Bone)
    RotationOffset=(X=-55,Y=0,Z=-20)
   )
   > RigVMVariableNode(Value=@n4.OutputParam) {variable=upperarm_r_rifle_up_alpha}
   > RigVMVariableNode(Value=&RigVMFunction_MathFloatMul(B=RigVMVariableNode() {variable=upperarm_r_rifle_up_alpha}.Value, A=RigVMVariableNode() {variable=head_SPR_lookup_alpha}.Value) {template="Multiply::Execute(in A,in B,out Result)"}.Result) {variable=look_up_two_hands_rifle_multiplied_alpha}
   AlphaInterp = &RigVMFunction_AlphaInterp(
    InterpSpeedDecreasing=10
    InterpSpeedIncreasing=10
    Scale=1
    bClampResult=true
    Value=RigVMVariableNode() {variable=look_up_two_hands_rifle_multiplied_alpha}.Value
   )
   > &RigUnit_ModifyTransforms(
    ItemToModify="((Item=(Type=Bone,Name=\"dyn_headBack_Base\"),Transform=(Rotation=(X=0,Y=0,Z=0,W=1),Translation=(X=0,Y=0,Z=0),Scale3D=(X=1,Y=1,Z=1))),(Item=(Type=Bone,Name=\"dyn_headBack_01_r\"),Transform=(Rotation=(X=0,Y=0,Z=0,W=1),Translation=(X=-3,Y=0,Z=0),Scale3D=(X=1,Y=1 ...(truncated)"
    Weight=@AlphaInterp.Result
    ItemToModify.1.Transform.Rotation=&RigVMFunction_MathQuaternionFromEuler(Euler=(X=0,Y=35,Z=-30)).Result
    ItemToModify.2.Transform.Rotation=&RigVMFunction_MathQuaternionFromEuler(Euler=(X=0,Y=0,Z=-15)).Result
   )
   n5 = &RigUnit_SphericalPoseReader(
    ActiveRegionScaleFactors=(PositiveWidth=1.5,NegativeWidth=1,PositiveHeight=1.5,NegativeHeight=1.5)
    ActiveRegionSize=0.1
    Debug=(bDrawDebug=True,DebugScale=25,DebugSegments=20,DebugThickness=0.25)
    DriverAxis=(X=0,Y=1,Z=0)
    DriverItem=(Type=Bone,Name="head")
    ExecutePin=()
    FalloffRegionScaleFactors=(PositiveWidth=2,NegativeWidth=0.5,PositiveHeight=1,NegativeHeight=1)
    FalloffSize=0.2
    OptionalParentItem=(Type=Bone)
    RotationOffset=(X=90,Y=0,Z=-30)
   )
   > RigVMVariableNode(Value=@n5.OutputParam) {variable=head_SPR_lookup_alpha}
   n6 = RigVMVariableNode() {variable=head_SPR_lookup_alpha}
   > RigUnit_SetRotation = &RigUnit_SetRotation(Item=(Type=Bone,Name="dyn_hair_bck_01"), Space=GlobalSpace, Value=(X=0,Y=0,Z=0,W=1), Weight=@n6.Value)
   > &RigUnit_SetRotation(Item=(Type=Bone,Name="dyn_hair_fr_l_01"), Space=GlobalSpace, Value=(X=0,Y=0,Z=0,W=1), Weight=@n6.Value)
   > &RigUnit_SetRotation(Item=(Type=Bone,Name="dyn_hair_fr_r_01"), Space=GlobalSpace, Value=(X=0,Y=0,Z=0,W=1), Weight=@n6.Value)
   > &RigUnit_SetRotation(Item=(Type=Bone,Name="dyn_hair_side_l_01"), Space=GlobalSpace, Value=(X=0,Y=0,Z=0,W=1), Weight=@n6.Value)
   > &RigUnit_SetRotation(Item=(Type=Bone,Name="dyn_hair_side_r_01"), Space=GlobalSpace, Value=(X=0,Y=0,Z=0,W=1), Weight=@n6.Value)
   > RigUnit_SetTranslation = /Script/ControlRig.RigUnit_SetTranslation(Item=(Type=Bone,Name="dyn_hair_bck_01"), Space=GlobalSpace, Value=(X=0,Y=-1,Z=1), bPropagateToChildren=True, Weight=@n6.Value) {is=struct RigVMUnitNode, template="Set Transform::Execute(in Item,in Space,in bInitial,in Value,in Weight,in bPropagateToChildren)"}
   /Script/RigVM.RigVMFunction_MathFloatMax(A=RigVMVariableNode() {variable=look_up_and_left_two_hands_multiplied_alpha}.Value) {is=struct RigVMUnitNode, template="Maximum::Execute(in A,in B,out Result)"}
   &RigUnit_SphericalPoseReader(
    ActiveRegionScaleFactors=(PositiveWidth=1,NegativeWidth=1,PositiveHeight=1,NegativeHeight=1)
    ActiveRegionSize=0.03
    Debug=(bDrawDebug=true,DebugScale=25,DebugSegments=20,DebugThickness=0.25)
    DriverAxis=(X=-1,Y=0,Z=0)
    DriverItem=(Type=Bone,Name="upperarm_r")
    ExecutePin=()
    FalloffRegionScaleFactors=(PositiveWidth=1,NegativeWidth=1,PositiveHeight=1,NegativeHeight=1)
    FalloffSize=0.125
    OptionalParentItem=(Type=Bone,Name="spine_04")
    RotationOffset=(X=-115,Y=0,Z=25)
   )
  ]}
]
hierarchy [
 root {loc=<0 0 0>, rot=<0 0 0 1>}
 pelvis {parent=root, loc=<0 -0.5301 98.0571>, rot=<0.0127 0.707 -0.0127 -0.707>}
 spine_01 {parent=pelvis, loc=<6.1208 -0.9328 0>, rot=<0 0 0.034 -0.9994>}
 spine_02 {parent=spine_01, loc=<6.5996 -0.2095 0>, rot=<0 0 0.0238 -0.9997>}
 spine_03 {parent=spine_02, loc=<6.4914 -0.1237 0>, rot=<0 0 0.0149 -0.9999>}
 spine_04 {parent=spine_03, loc=<6.9598 0.0794 0>, rot=<0 0 -0.0132 -0.9999>}
 spine_05 {parent=spine_04, loc=<6.9987 0 0>, rot=<0 0 -0.0015 -1>}
 clavicle_l {parent=spine_05, loc=<5.4831 -0.1053 -3.2946>, rot=<0.0465 -0.7289 -0.009 -0.6829>}
 upperarm_l {parent=clavicle_l, loc=<9.8658 -1.2136 -0.4105>, rot=<0.0164 -0.3245 -0.0269 -0.9454>}
 lowerarm_l {parent=upperarm_l, loc=<26.465 -0.8345 0.259>, rot=<0.0036 0.0161 -0.2298 -0.9731>}
 hand_l {parent=lowerarm_l, loc=<22.7355 -0.3203 -0.6105>, rot=<-0.6784 -0.0223 0.0735 -0.7307>}
 thumb_01_l {parent=hand_l, loc=<1.3118 -0.3191 -2.459>, rot=<0.6546 -0.4042 0.0664 -0.6354>}
 thumb_02_l {parent=thumb_01_l, loc=<5.9336 -0.0448 0.0401>, rot=<-0.0037 0.0213 0.298 -0.9543>}
 thumb_03_l {parent=thumb_02_l, loc=<2.2223 0.3682 -0.0907>, rot=<0.0118 -0.013 0.0066 -0.9998>}
 index_metacarpal_l {parent=hand_l, loc=<1.6535 0.0183 -2.96>, rot=<0.0158 -0.0124 -0.0173 -0.9996>}
 index_01_l {parent=index_metacarpal_l, loc=<8.3008 0.0357 -0.0524>, rot=<-0.0044 0.0485 0.0924 -0.9945>}
 index_02_l {parent=index_01_l, loc=<3.3587 -0.3004 0.0452>, rot=<-0.002 0.0027 0.3225 -0.9466>}
 index_03_l {parent=index_02_l, loc=<2.5578 -0.123 -0.0252>, rot=<-0.002 0.0067 0.2558 -0.9667>}
 middle_metacarpal_l {parent=hand_l, loc=<2.1856 -0.1725 -0.5984>, rot=<0.006 0.0168 -0.0329 -0.9993>}
 middle_01_l {parent=middle_metacarpal_l, loc=<7.3275 -0.2447 0.0129>, rot=<-0.0208 0.0042 0.0858 -0.9961>}
 middle_02_l {parent=middle_01_l, loc=<4.0123 0.1401 0.1175>, rot=<-0.0218 0.0258 0.3512 -0.9357>}
 middle_03_l {parent=middle_02_l, loc=<2.6077 0.0534 0.1373>, rot=<0.0082 0.0188 0.1708 -0.9851>}
 ring_metacarpal_l {parent=hand_l, loc=<2.2074 -0.3625 1.3638>, rot=<-0.0284 0.0722 -0.0458 -0.9959>}
 ring_01_l {parent=ring_metacarpal_l, loc=<7.0295 -0.6197 -0.1446>, rot=<-0.0489 -0.0252 0.1892 -0.9804>}
 ring_02_l {parent=ring_01_l, loc=<3.6174 0.4647 -0.1027>, rot=<0 0 0.2806 -0.9598>}
 ring_03_l {parent=ring_02_l, loc=<2.4152 0.114 0.0199>, rot=<0 0 0.1189 -0.9929>}
 pinky_metacarpal_l {parent=hand_l, loc=<1.7616 -0.9753 3.3814>, rot=<-0.0397 0.0814 -0.069 -0.9935>}
 pinky_01_l {parent=pinky_metacarpal_l, loc=<7.1709 -0.4831 -0.1094>, rot=<-0.0654 -0.0035 0.2242 -0.9723>}
 pinky_02_l {parent=pinky_01_l, loc=<2.8274 0.0893 -0.1043>, rot=<0.0037 0.0011 0.2854 -0.9584>}
 pinky_03_l {parent=pinky_02_l, loc=<2.2046 0.0307 -0.0078>, rot=<-0.0002 0.0098 0.1445 -0.9895>}
 lowerarm_twist_02_l {parent=lowerarm_l, loc=<10.6707 0 0>, rot=<0 0 0 1>}
 lowerarm_twist_01_l {parent=lowerarm_l, loc=<18.316 0 0>, rot=<0 0 0 1>}
 upperarm_twist_01_l {parent=upperarm_l, loc=<7.2135 -0.1134 0>, rot=<0 0 0 1>}
 upperarm_twist_02_l {parent=upperarm_l, loc=<19.7998 -0.1134 0>, rot=<0 0 0 1>}
 neck_01 {parent=spine_05, loc=<11.3138 0.1529 0>, rot=<0 0 -0.2003 -0.9797>}
 neck_02 {parent=neck_01, loc=<4.5627 0 0>, rot=<0 0 0.0504 -0.9987>}
 head {parent=neck_02, loc=<5.3914 0 0>, rot=<0 0 0.115 -0.9934>}
 dyn_hatHair_MSTR {parent=head, loc=<11.6562 -0.6479 0>, rot=<0.7416 0.1976 0.5941 0.2411>}
 dyn_ear_r {parent=dyn_hatHair_MSTR, loc=<9.5331 0 0>, rot=<0 0 0 1>}
 dyn_ear_l {parent=dyn_hatHair_MSTR, loc=<-2.2395 -8.8746 2.6655>, rot=<0.1504 0.0868 0.8529 -0.4924>}
 dyn_pompom_01 {parent=dyn_hatHair_MSTR, loc=<-8.5685 8.1929 1.6473>, rot=<-0.4279 0.1672 0.8641 0.2058>}
 dyn_pompom_02 {parent=dyn_pompom_01, loc=<7.3133 0 0>, rot=<0 0 0 1>}
 dyn_frStrand_l_01 {parent=dyn_hatHair_MSTR, loc=<0.6976 -9.7562 -4.5147>, rot=<-0.3606 -0.0687 0.9246 -0.1016>}
 dyn_frStrand_l_02 {parent=dyn_frStrand_l_01, loc=<4.0775 0 0>, rot=<-0.2676 -0.0719 0.1563 -0.948>}
 dyn_frStrand_l_03 {parent=dyn_frStrand_l_02, loc=<3.918 0 0>, rot=<0 0.1736 0 -0.9848>}
 dyn_frStrand_r_01 {parent=dyn_hatHair_MSTR, loc=<7.7557 -1.1079 -6.9504>, rot=<0.3171 -0.5204 -0.5541 -0.5672>}
 dyn_frStrand_r_02 {parent=dyn_frStrand_r_01, loc=<5.2085 0 0>, rot=<0.896 0.0374 -0.1955 0.3968>}
 dyn_frStrand_r_03 {parent=dyn_frStrand_r_02, loc=<4.7122 0 0>, rot=<-0.0651 0.2064 0.2936 -0.9311>}
 dyn_hair_fr_l_01 {parent=dyn_hatHair_MSTR, loc=<-8.0353 -5.5191 -9.1766>, rot=<-0.1693 -0.7112 0.2392 -0.6391>}
 dyn_hair_fr_l_02 {parent=dyn_hair_fr_l_01, loc=<7.3047 0 0>, rot=<0 0 0 1>}
 dyn_hair_fr_r_01 {parent=dyn_hatHair_MSTR, loc=<5.0677 4.1626 -11.6149>, rot=<0.006 -0.7112 -0.0094 -0.7029>}
 dyn_hair_fr_r_02 {parent=dyn_hair_fr_r_01, loc=<7.9174 0 0>, rot=<0 0 0 1>}
 dyn_hair_bck_MSTR {parent=dyn_hatHair_MSTR, loc=<-7.0544 6.8465 -9.702>, rot=<-0.4503 0.3359 0.7492 0.3507>}
 dyn_hair_bck_01 {parent=dyn_hair_bck_MSTR, loc=<2.2476 0 0>, rot=<0 0 0 1>}
 dyn_hair_bck_02 {parent=dyn_hair_bck_01, loc=<5.776 0 0>, rot=<0 0 0 1>}
 dyn_hair_side_l_MSTR {parent=dyn_hatHair_MSTR, loc=<-9.4467 -0.4311 -8.5368>, rot=<-0.4083 0.0052 0.9128 0.0069>}
 dyn_hair_side_l_01 {parent=dyn_hair_side_l_MSTR, loc=<3.0769 0 0>, rot=<0 0 0 1>}
 dyn_hair_side_l_02 {parent=dyn_hair_side_l_01, loc=<5.5799 0 0>, rot=<0 0 0 1>}
 dyn_hair_side_r_MSTR {parent=dyn_hatHair_MSTR, loc=<0.1523 8.6991 -10.7767>, rot=<0.3018 -0.5798 -0.4819 -0.5835>}
 dyn_hair_side_r_01 {parent=dyn_hair_side_r_MSTR, loc=<0 0 0>, rot=<0 0 0 1>}
 dyn_hair_side_r_02 {parent=dyn_hair_side_r_01, loc=<7.9756 0 0>, rot=<0 0 0 1>}
 dyn_headBack_Base {parent=head, loc=<6.9539 -5.7482 -0.0677>, rot=<0.1807 -0.9835 0 0>}
 dyn_headBack_01_c {parent=dyn_headBack_Base, loc=<5.269 -0.4697 -0.1062>, rot=<0 0 -0.0203 -0.9998>}
 dyn_headBack_02_c {parent=dyn_headBack_01_c, loc=<3.9528 0 0>, rot=<0 0 -0.0156 -0.9999>}
 dyn_headBack_03_c {parent=dyn_headBack_02_c, loc=<4.6852 0 0>, rot=<0 0 0 1>}
 dyn_headBack_01_l {parent=dyn_headBack_Base, loc=<4.8669 1.4907 5.4464>, rot=<0.4407 0.2004 0.0062 -0.875>}
 dyn_headBack_02_l {parent=dyn_headBack_01_l, loc=<4.1345 0 0>, rot=<-0.0393 -0.0144 0.0264 -0.9988>}
 dyn_headBack_03_l {parent=dyn_headBack_02_l, loc=<4.5827 0 0>, rot=<0 0 0 1>}
 dyn_headBack_01_r {parent=dyn_headBack_Base, loc=<4.6665 2.2033 -5.0424>, rot=<-0.1882 -0.0662 -0.0164 -0.9798>}
 dyn_headBack_02_r {parent=dyn_headBack_01_r, loc=<4.1345 0 0>, rot=<-0.1285 -0.0538 0.0742 -0.9875>}
 dyn_headBack_03_r {parent=dyn_headBack_02_r, loc=<4.5827 0 0>, rot=<0 0 0 1>}
 dyn_hairFrontBounce_01_Base {parent=head, loc=<2.7792 4.7152 7.7078>, rot=<-0.001 0.0201 0.9989 0.0413>}
 dyn_hairFrontBounce_01 {parent=dyn_hairFrontBounce_01_Base, loc=<0 0 0>, rot=<0 0 0 1>}
 dyn_hairFrontBounce_02 {parent=dyn_hairFrontBounce_01, loc=<5.6405 0 0>, rot=<-0.3719 -0.3508 0.5305 -0.6761>}
 dyn_headFront_01_Base_r {parent=head, loc=<2.7793 2.2241 6.1332>, rot=<0.0631 -0.884 0.4623 0.0292>}
 dyn_headFront_01_r {parent=dyn_headFront_01_Base_r, loc=<0 0 0>, rot=<0 0 0 1>}
 dyn_headFront_02_r {parent=dyn_headFront_01_r, loc=<5.4478 0 0>, rot=<-0.5562 -0.2621 0.0196 -0.7884>}
 dyn_headFront_03_r {parent=dyn_headFront_02_r, loc=<4.6029 0 0>, rot=<-0.0105 0.1028 0.6868 -0.7194>}
 dyn_headFront_01_Base_l {parent=head, loc=<3.2526 3.9383 -7.1179>, rot=<-0.1444 0.6533 0.7368 -0.0971>}
 dyn_headFront_01_l {parent=dyn_headFront_01_Base_l, loc=<0 0 0>, rot=<0 0 0 1>}
 dyn_headFront_02_l {parent=dyn_headFront_01_l, loc=<3.9425 0 0>, rot=<0.0227 0.1633 -0.1007 -0.9812>}
 dyn_headFront_03_l {parent=dyn_headFront_02_l, loc=<4.3209 0 0>, rot=<0.1723 0.2219 0.1738 -0.9439>}
 dyn_headFront_04_l {parent=dyn_headFront_03_l, loc=<2.9219 0 0>, rot=<0.09 -0.1263 0.6771 -0.7194>}
 dyn_head_01_Base_l {parent=head, loc=<3.1271 -1.7629 -8.368>, rot=<0.0456 -0.7633 -0.6422 0.0531>}
 dyn_head_01_l {parent=dyn_head_01_Base_l, loc=<0 0 0>, rot=<0 0 0 1>}
 dyn_head_02_l {parent=dyn_head_01_l, loc=<5.3391 0 0>, rot=<0.0863 0.0823 0.1845 -0.9756>}
 dyn_head_03_l {parent=dyn_head_02_l, loc=<4.1111 0 0>, rot=<0 0 0 1>}
 dyn_head_01_Base_r {parent=head, loc=<2.3393 -2.5791 6.7725>, rot=<0.066 -0.704 0.7045 0.0602>}
 dyn_head_01_r {parent=dyn_head_01_Base_r, loc=<0 0 0>, rot=<0 0 0 1>}
 dyn_head_02_r {parent=dyn_head_01_r, loc=<5.3391 0 0>, rot=<0.0786 0.0273 0.1549 -0.9844>}
 dyn_head_03_r {parent=dyn_head_02_r, loc=<4.1111 0 0>, rot=<0 0 0 1>}
 clavicle_r {parent=spine_05, loc=<5.4831 -0.1053 3.2946>, rot=<0.7289 0.0465 -0.6829 0.009>}
 upperarm_r {parent=clavicle_r, loc=<-9.8658 1.2136 0.4105>, rot=<0.0164 -0.3245 -0.0269 -0.9454>}
 lowerarm_r {parent=upperarm_r, loc=<-26.465 0.8345 -0.259>, rot=<0.0036 0.0161 -0.2298 -0.9731>}
 hand_r {parent=lowerarm_r, loc=<-22.7355 0.3203 0.6106>, rot=<-0.6784 -0.0223 0.0735 -0.7307>}
 thumb_01_r {parent=hand_r, loc=<-1.3119 0.3192 2.459>, rot=<0.6546 -0.4042 0.0664 -0.6354>}
 thumb_02_r {parent=thumb_01_r, loc=<-5.9336 0.0448 -0.0401>, rot=<-0.0037 0.0213 0.298 -0.9543>}
 thumb_03_r {parent=thumb_02_r, loc=<-2.2223 -0.3682 0.0907>, rot=<0.0118 -0.013 0.0066 -0.9998>}
 index_metacarpal_r {parent=hand_r, loc=<-1.6534 -0.0181 2.9601>, rot=<0.0158 -0.0124 -0.0173 -0.9996>}
 index_01_r {parent=index_metacarpal_r, loc=<-8.3008 -0.0357 0.0524>, rot=<-0.0044 0.0485 0.0924 -0.9945>}
 index_02_r {parent=index_01_r, loc=<-3.3589 0.3003 -0.0453>, rot=<-0.002 0.0027 0.3225 -0.9466>}
 index_03_r {parent=index_02_r, loc=<-2.5578 0.123 0.0252>, rot=<-0.002 0.0067 0.2558 -0.9667>}
 middle_metacarpal_r {parent=hand_r, loc=<-2.1856 0.1726 0.5984>, rot=<0.006 0.0168 -0.0329 -0.9993>}
 middle_01_r {parent=middle_metacarpal_r, loc=<-7.3273 0.2447 -0.0129>, rot=<-0.0208 0.0042 0.0858 -0.9961>}
 middle_02_r {parent=middle_01_r, loc=<-4.0124 -0.14 -0.1175>, rot=<-0.0218 0.0258 0.3512 -0.9357>}
 middle_03_r {parent=middle_02_r, loc=<-2.6077 -0.0534 -0.1373>, rot=<0.0082 0.0188 0.1708 -0.9851>}
 ring_metacarpal_r {parent=hand_r, loc=<-2.2073 0.3627 -1.3638>, rot=<-0.0284 0.0722 -0.0458 -0.9959>}
 ring_01_r {parent=ring_metacarpal_r, loc=<-7.0297 0.6197 0.1446>, rot=<-0.0489 -0.0252 0.1892 -0.9804>}
 ring_02_r {parent=ring_01_r, loc=<-3.6171 -0.4648 0.1027>, rot=<0 0 0.2806 -0.9598>}
 ring_03_r {parent=ring_02_r, loc=<-2.4152 -0.114 -0.0199>, rot=<0 0 0.1189 -0.9929>}
 pinky_metacarpal_r {parent=hand_r, loc=<-1.7617 0.9753 -3.3814>, rot=<-0.0397 0.0814 -0.069 -0.9935>}
 pinky_01_r {parent=pinky_metacarpal_r, loc=<-7.1707 0.4832 0.1094>, rot=<-0.0654 -0.0035 0.2242 -0.9723>}
 pinky_02_r {parent=pinky_01_r, loc=<-2.8274 -0.0893 0.1043>, rot=<0.0037 0.0011 0.2854 -0.9584>}
 pinky_03_r {parent=pinky_02_r, loc=<-2.2045 -0.031 0.0078>, rot=<-0.0002 0.0098 0.1445 -0.9895>}
 lowerarm_twist_02_r {parent=lowerarm_r, loc=<-10.6707 0 0>, rot=<0 0 0 -1>}
 lowerarm_twist_01_r {parent=lowerarm_r, loc=<-18.316 0 0>, rot=<0 0 0 -1>}
 upperarm_twist_01_r {parent=upperarm_r, loc=<-7.2135 0.1134 0>, rot=<0 0 0 -1>}
 upperarm_twist_02_r {parent=upperarm_r, loc=<-19.7998 0.1134 0>, rot=<0 0 0 -1>}
 thigh_l {parent=pelvis, loc=<-1.0855 -0.2109 -8.7053>, rot=<0.0009 0.0355 0.0155 -0.9993>}
 calf_l {parent=thigh_l, loc=<-42.5284 -0.0019 -0.7297>, rot=<0.0006 -0.0161 -0.0391 -0.9991>}
 foot_l {parent=calf_l, loc=<-46.7401 -2.5014 -0.592>, rot=<0.0427 -0.0238 0.0353 -0.9982>}
 ball_l {parent=foot_l, loc=<-3.5658 10.8066 -0.1544>, rot=<0.0001 -0.0001 -0.7063 -0.7079>}
 calf_twist_02_l {parent=calf_l, loc=<-19.1546 -1.7943 -0.2638>, rot=<0.0237 0.0095 -0.0035 -0.9997>}
 calf_twist_01_l {parent=calf_l, loc=<-35.8483 -1.9197 -0.5794>, rot=<0.0237 0.0095 -0.0035 -0.9997>}
 thigh_twist_01_l {parent=thigh_l, loc=<-27.952 0 0>, rot=<0.0013 0 0 -1>}
 thigh_r {parent=pelvis, loc=<-1.0853 -0.2109 8.7053>, rot=<0.0355 -0.0009 0.9993 0.0155>}
 calf_r {parent=thigh_r, loc=<42.5286 0.0019 0.7297>, rot=<0.0006 -0.0161 -0.0391 -0.9991>}
 foot_r {parent=calf_r, loc=<46.7403 2.5014 0.592>, rot=<0.0427 -0.0238 0.0353 -0.9982>}
 ball_r {parent=foot_r, loc=<3.5658 -10.8066 0.1544>, rot=<0.0001 -0.0001 -0.7063 -0.7079>}
 calf_twist_02_r {parent=calf_r, loc=<19.1546 1.7943 0.2638>, rot=<0.0237 0.0095 -0.0035 -0.9997>}
 calf_twist_01_r {parent=calf_r, loc=<35.8485 1.9197 0.5794>, rot=<0.0237 0.0095 -0.0035 -0.9997>}
 thigh_twist_01_r {parent=thigh_r, loc=<27.952 0 0>, rot=<0.0013 0 0 -1>}
]
variables [
 NewVar {type=double, default=0}
 head_SPR_lookup_alpha {type=double, default=0}
 look_up_and_left_two_hands_multiplied_alpha {type=double, default=0}
 look_up_two_hands_rifle_multiplied_alpha {type=double, default=0}
 uppearm_r_SPR_aim_up_two_hands_alpha {type=double, default=0}
 upperarm_r_rifle_up_alpha {type=double, default=0}
]
