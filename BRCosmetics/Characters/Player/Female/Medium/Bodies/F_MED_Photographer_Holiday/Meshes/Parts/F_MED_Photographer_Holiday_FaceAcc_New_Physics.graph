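# Physics asset for the Holiday female Photographer's face accessory mesh, setting up secondary motion for what bone names indicate is hair.
# Capsules on head, neck, clavicles and upper spine form the body volume around the dynamic bone chains.
# Front strands hang from a hat-hair master locked to the head; back, front and side chains start from base bones with free angular constraints to the head.
# Most chain links have narrow swing and twist limits.
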PhysicsAsset /BRCosmetics/Characters/Player/Female/Medium/Bodies/F_MED_Photographer_Holiday/Meshes/Parts/F_MED_Photographer_Holiday_FaceAcc_New_Physics
bodies {
 clavicle_l [{shapeType=Capsule, length=8.25, radius=5.25, shapeName=clavicle_l_capsule, center=<5.5926 0 0>, rotation=<rot 90 -179.789 -179.7891>}]
 clavicle_r [{shapeType=Capsule, length=8.25, radius=5.25, shapeName=clavicle_r_capsule, center=<-5.698 0 0.2659>, rotation=<rot -80 179.9994 -179.9998>}]
 dyn_frStrand_l_01 [{
   shapeType=Capsule
   length=2.7792
   radius=1.2136
   shapeName=dyn_frStrand_l_01_capsule
   center=<1.0067 -0.1067 -0.0745>
   rotation=<rot 0 89.2486 -77.245>
  }]
 dyn_frStrand_l_02 [{shapeType=Capsule, length=1, radius=1, shapeName=dyn_frStrand_l_02_capsule, rotation=<rot 80.0001 -179.9998 -179.9999>}]
 dyn_frStrand_l_03 [{shapeType=Capsule, length=1, radius=1, shapeName=dyn_frStrand_l_03_capsule, rotation=<rot 80 0 0>}]
 dyn_frStrand_r_01 [{shapeType=Capsule, length=1, radius=1, shapeName=dyn_frStrand_r_01_capsule, rotation=<rot 90 18.0327 18.0329>}]
 dyn_frStrand_r_02 [{shapeType=Capsule, length=1, radius=1, shapeName=dyn_frStrand_r_02_capsule, rotation=<rot -46.0419 119.5199 -112.1756>}]
 dyn_frStrand_r_03 [{
   shapeType=Capsule
   length=1.5
   radius=1.75
   shapeName=dyn_frStrand_r_03_capsule
   center=<-0.1552 -0.7497 0.8799>
   rotation=<rot -48.9734 78.3078 -74.66>
  }]
 dyn_hatHair_MSTR [{
   shapeType=Capsule
   length=0
   radius=4.2841
   shapeName=dyn_hatHair_MSTR_capsule
   center=<-0.9993 1.7903 1.3884>
   rotation=<rot -9.8365 44.0162 -79.5198>
  }]
 dyn_headBack_01_c [{shapeType=Capsule, length=0.505, radius=0.505, shapeName=dyn_headBack_01_c_capsule, rotation=<rot 0 -90 0>}]
 dyn_headBack_01_l [{shapeType=Capsule, length=4, radius=2.5, shapeName=dyn_headBack_01_l_capsule, center=<2 -1 0>, rotation=<rot 0 0 -20>}]
 dyn_headBack_01_r [{shapeType=Capsule, length=4, radius=2.5, shapeName=dyn_headBack_01_r_capsule, center=<2 -2 0>, rotation=<rot 0 0.0001 -15>}]
 dyn_headBack_02_c [{shapeType=Capsule, length=4, radius=2.5, shapeName=dyn_headBack_02_c_capsule, center=<2 -1.5 0>}]
 dyn_headBack_02_l [{shapeType=Capsule, length=4, radius=2.75, shapeName=dyn_headBack_02_l_capsule, center=<2 -1 0>}]
 dyn_headBack_02_r [{shapeType=Capsule, length=3.7066, radius=2.7934, shapeName=dyn_headBack_02_r_capsule, center=<2 -1 0>}]
 dyn_headBack_03_c [{shapeType=Capsule, length=4, radius=2.5, shapeName=dyn_headBack_03_c_capsule, center=<2 -2.5 0>}]
 dyn_headBack_03_l [{shapeType=Capsule, length=4, radius=2.75, shapeName=dyn_headBack_03_l_capsule, center=<2 -1 0>}]
 dyn_headBack_03_r [{shapeType=Capsule, length=3.3504, radius=3.1496, shapeName=dyn_headBack_03_r_capsule, center=<2 -2 0>}]
 dyn_headBack_Base [{shapeType=Capsule, length=0.505, radius=0.505, shapeName=dyn_headBack_Base_capsule, rotation=<rot 0 -90 0>}]
 dyn_headFront_01_Base_l [{shapeType=Capsule, length=0.505, radius=0.505, shapeName=dyn_headFront_01_Base_l_capsule, rotation=<rot 0 -90 0>}]
 dyn_headFront_01_Base_r [{shapeType=Capsule, length=0.505, radius=0.505, shapeName=dyn_headFront_01_Base_r_capsule, rotation=<rot 0 -90 0>}]
 dyn_headFront_01_l [{
   shapeType=Capsule
   length=2.0976
   radius=1.6837
   shapeName=dyn_headFront_01_l_capsule
   center=<4.8776 -1.3122 -1.4177>
   rotation=<rot 0 104.9289 -82.7856>
  }]
 dyn_headFront_01_r [{shapeType=Capsule, length=3, radius=2.5, shapeName=dyn_headFront_01_r_capsule, center=<2.6489 -2.5 -2>, rotation=<rot -0.0002 0.0002 -55>}]
 dyn_headFront_02_l [{shapeType=Capsule, length=3, radius=2.5, shapeName=dyn_headFront_02_l_capsule, center=<2 -2.5 0>, rotation=<rot 0.0001 0 20>}]
 dyn_headFront_02_r [{shapeType=Capsule, length=3, radius=2.5, shapeName=dyn_headFront_02_r_capsule, center=<2 -2 1>, rotation=<rot -0.0001 -0.0001 15>}]
 dyn_headFront_03_l [{shapeType=Capsule, length=3, radius=2.5, shapeName=dyn_headFront_03_l_capsule, center=<2 -2 0>}]
 dyn_headFront_03_r [{shapeType=Capsule, length=1.4073, radius=0.8308, shapeName=dyn_headFront_03_r_capsule, center=<0 1 0>}]
 dyn_headFront_04_l [{shapeType=Capsule, length=1.9276, radius=0.8535, shapeName=dyn_headFront_04_l_capsule, center=<0 1 0>}]
 dyn_head_01_Base_l [{shapeType=Capsule, length=0.505, radius=0.505, shapeName=dyn_head_01_Base_l_capsule, rotation=<rot 0 -90 0>}]
 dyn_head_01_Base_r [{shapeType=Capsule, length=0.505, radius=0.505, shapeName=dyn_head_01_Base_r_capsule, rotation=<rot 0 -90 0>}]
 dyn_head_01_l [{shapeType=Capsule, length=2.8442, radius=1.5509, shapeName=dyn_head_01_l_capsule, center=<2 0 0>}]
 dyn_head_01_r [{shapeType=Capsule, length=0.505, radius=0.505, shapeName=dyn_head_01_r_capsule, center=<2 0 0>}]
 dyn_head_02_l [{shapeType=Capsule, length=4, radius=2.5, shapeName=dyn_head_02_l_capsule, center=<2 -1 0>}]
 dyn_head_02_r [{shapeType=Capsule, length=4, radius=2.5, shapeName=dyn_head_02_r_capsule, center=<2 -1 0>}]
 dyn_head_03_l [{shapeType=Capsule, length=4, radius=2.5, shapeName=dyn_head_03_l_capsule, center=<2 -1 0>}]
 dyn_head_03_r [{shapeType=Capsule, length=4, radius=2.5, shapeName=dyn_head_03_r_capsule, center=<2 -1 0>}]
 head [{shapeType=Capsule, length=5.1802, radius=7.8371, shapeName=head_capsule, center=<5.0549 0.7149 0>, rotation=<rot 90 0 0>}]
 neck_02 [{shapeType=Capsule, length=8.25, radius=7.25, shapeName=neck_02_capsule, center=<6.2244 -0.1867 0>, rotation=<rot 90 178.2101 -171.7895>}]
 spine_05 [
  {shapeType=Capsule, length=4.4093, radius=8.3407, shapeName=spine_05_capsule, center=<7.9296 -1.2225 0>}
  {shapeType=Capsule, length=5.5076, radius=7.9924, shapeName=spine_05_capsule, center=<9.7868 1.4366 0.1323>, rotation=<rot 0 9.9999 -90>}
 ]
}
constraints [
 {bone1Name=dyn_frStrand_l_02, bone2Name=dyn_frStrand_l_01, swing1LimitDegrees=10, swing2LimitDegrees=30, twistLimitDegrees=1}
 {bone1Name=dyn_frStrand_l_03, bone2Name=dyn_frStrand_l_02, swing1LimitDegrees=10, swing2LimitDegrees=30, twistLimitDegrees=1}
 {bone1Name=dyn_frStrand_r_01, bone2Name=dyn_hatHair_MSTR, swing1LimitDegrees=10, swing2LimitDegrees=20, twistLimitDegrees=1}
 {bone1Name=dyn_frStrand_r_02, bone2Name=dyn_frStrand_r_01, swing1LimitDegrees=10, swing2LimitDegrees=20, twistLimitDegrees=1}
 {bone1Name=dyn_frStrand_r_03, bone2Name=dyn_frStrand_r_02, swing1LimitDegrees=10, swing2LimitDegrees=20, twistLimitDegrees=1}
 {
  bone1Name=dyn_hatHair_MSTR
  bone2Name=head
  swing1LimitDegrees=45
  swing1Motion=Locked
  swing2LimitDegrees=45
  swing2Motion=Locked
  twistLimitDegrees=45
  twistMotion=Locked
 }
 {
  bone1Name=dyn_frStrand_l_01
  bone2Name=dyn_hatHair_MSTR
  swing1LimitDegrees=45
  swing1Motion=Free
  swing2LimitDegrees=45
  swing2Motion=Free
  twistLimitDegrees=45
  twistMotion=Free
 }
 {bone1Name=dyn_head_03_r, bone2Name=dyn_head_02_r, swing1LimitDegrees=10, swing2LimitDegrees=10, twistLimitDegrees=1}
 {
  bone1Name=dyn_headBack_Base
  bone2Name=head
  swing1LimitDegrees=45
  swing1Motion=Free
  swing2LimitDegrees=45
  swing2Motion=Free
  twistLimitDegrees=45
  twistMotion=Free
 }
 {
  bone1Name=dyn_headBack_01_c
  bone2Name=dyn_headBack_Base
  swing1LimitDegrees=45
  swing1Motion=Free
  swing2LimitDegrees=45
  swing2Motion=Free
  twistLimitDegrees=45
  twistMotion=Free
 }
 {bone1Name=dyn_headBack_02_c, bone2Name=dyn_headBack_01_c, swing1LimitDegrees=20, swing2LimitDegrees=10, twistLimitDegrees=1}
 {bone1Name=dyn_headBack_03_c, bone2Name=dyn_headBack_02_c, swing1LimitDegrees=20, swing2LimitDegrees=10, twistLimitDegrees=1}
 {bone1Name=dyn_headBack_01_l, bone2Name=dyn_headBack_Base, swing1LimitDegrees=20, swing2LimitDegrees=10, twistLimitDegrees=1}
 {bone1Name=dyn_headBack_02_l, bone2Name=dyn_headBack_01_l, swing1LimitDegrees=20, swing2LimitDegrees=20, twistLimitDegrees=1}
 {bone1Name=dyn_headBack_03_l, bone2Name=dyn_headBack_02_l, swing1LimitDegrees=20, swing2LimitDegrees=20, twistLimitDegrees=1}
 {bone1Name=dyn_headBack_01_r, bone2Name=dyn_headBack_Base, swing1LimitDegrees=20, swing2LimitDegrees=10, twistLimitDegrees=1}
 {bone1Name=dyn_headBack_02_r, bone2Name=dyn_headBack_01_r, swing1LimitDegrees=30, swing2LimitDegrees=10, twistLimitDegrees=1}
 {bone1Name=dyn_headBack_03_r, bone2Name=dyn_headBack_02_r, swing1LimitDegrees=30, swing2LimitDegrees=10, twistLimitDegrees=1}
 {
  bone1Name=dyn_headFront_01_Base_l
  bone2Name=head
  swing1LimitDegrees=45
  swing1Motion=Free
  swing2LimitDegrees=45
  swing2Motion=Free
  twistLimitDegrees=45
  twistMotion=Free
 }
 {
  bone1Name=dyn_headFront_01_l
  bone2Name=dyn_headFront_01_Base_l
  swing1LimitDegrees=45
  swing1Motion=Free
  swing2LimitDegrees=45
  swing2Motion=Free
  twistLimitDegrees=45
  twistMotion=Free
 }
 {bone1Name=dyn_headFront_02_l, bone2Name=dyn_headFront_01_l, swing1LimitDegrees=20, swing2LimitDegrees=10, twistLimitDegrees=1}
 {bone1Name=dyn_headFront_03_l, bone2Name=dyn_headFront_02_l, swing1LimitDegrees=20, swing2LimitDegrees=10, twistLimitDegrees=1}
 {bone1Name=dyn_headFront_04_l, bone2Name=dyn_headFront_03_l, swing1LimitDegrees=20, swing2LimitDegrees=10, twistLimitDegrees=1}
 {
  bone1Name=dyn_headFront_01_Base_r
  bone2Name=head
  swing1LimitDegrees=45
  swing1Motion=Free
  swing2LimitDegrees=45
  swing2Motion=Free
  twistLimitDegrees=45
  twistMotion=Free
 }
 {bone1Name=dyn_headFront_01_r, bone2Name=dyn_headFront_01_Base_r, swing1LimitDegrees=10, swing2LimitDegrees=10, twistLimitDegrees=1}
 {bone1Name=dyn_headFront_02_r, bone2Name=dyn_headFront_01_r, swing1LimitDegrees=10, swing2LimitDegrees=10, twistLimitDegrees=1}
 {bone1Name=dyn_headFront_03_r, bone2Name=dyn_headFront_02_r, swing1LimitDegrees=10, swing2LimitDegrees=10, twistLimitDegrees=1}
 {
  bone1Name=dyn_head_01_Base_l
  bone2Name=head
  swing1LimitDegrees=45
  swing1Motion=Free
  swing2LimitDegrees=45
  swing2Motion=Free
  twistLimitDegrees=45
  twistMotion=Free
 }
 {
  bone1Name=dyn_head_01_l
  bone2Name=dyn_head_01_Base_l
  swing1LimitDegrees=45
  swing1Motion=Free
  swing2LimitDegrees=45
  swing2Motion=Free
  twistLimitDegrees=45
  twistMotion=Free
 }
 {bone1Name=dyn_head_02_l, bone2Name=dyn_head_01_l, swing1LimitDegrees=20, swing2LimitDegrees=10, twistLimitDegrees=1}
 {bone1Name=dyn_head_03_l, bone2Name=dyn_head_02_l, swing1LimitDegrees=20, swing2LimitDegrees=10, twistLimitDegrees=1}
 {
  bone1Name=dyn_head_01_Base_r
  bone2Name=head
  swing1LimitDegrees=45
  swing1Motion=Free
  swing2LimitDegrees=45
  swing2Motion=Free
  twistLimitDegrees=45
  twistMotion=Free
 }
 {
  bone1Name=dyn_head_01_r
  bone2Name=dyn_head_01_Base_r
  swing1LimitDegrees=45
  swing1Motion=Free
  swing2LimitDegrees=45
  swing2Motion=Free
  twistLimitDegrees=45
  twistMotion=Free
 }
 {bone1Name=dyn_head_02_r, bone2Name=dyn_head_01_r, swing1LimitDegrees=10, swing2LimitDegrees=10, twistLimitDegrees=1}
]
