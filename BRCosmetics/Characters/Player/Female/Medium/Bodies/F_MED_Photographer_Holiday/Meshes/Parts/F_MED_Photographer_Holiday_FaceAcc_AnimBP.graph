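# Post-process style AnimBlueprint for the Holiday Photographer face accessory, adding secondary motion to hair, hat pompoms and ears.
# It copies the attached parent's pose through a Control Rig, then two RigidBody nodes switch physics assets on bIsInFrontEnd.
# Additive bone offsets, AnimDynamics on pompoms and ears, and a pompom constraint follow.
# The EventGraph computes aim, arm, crouch and gravity values that the AnimGraph never reads; several AnimGraph nodes are unconnected.
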AnimBlueprint /BRCosmetics/Characters/Player/Female/Medium/Bodies/F_MED_Photographer_Holiday/Meshes/Parts/F_MED_Photographer_Holiday_FaceAcc_AnimBP
target_skeleton=&/F_MED_Photographer_Holiday_FaceAcc_Skeleton
graphs [
 AnimGraph {nodes [
   "Constraint - dyn_pompom_01 (PP)"!(ComponentPose, Alpha=0.3, ConstraintWeights_0=0.5442, ConstraintWeights_1:float=0.4558) -> (Pose) {
    alphaBoolBlend {blendOption=Linear}
    boneToModify {boneName=dyn_pompom_01}
    constraintSetup [{targetBone {boneName=dyn_hatHair_MSTR}}, {targetBone {boneName=dyn_pompom_01}}]
    constraintWeights [1,1]
   }
   "Transform (Modify) Bone - Bone: dyn_hair_bck_01"(Alpha=1, Translation="0,0,0", Rotation="15,0,0", Scale="1,1,1") {
    rotationMode=BMM_Additive
    rotationSpace=BCS_BoneSpace
    scaleSpace=BCS_BoneSpace
    translationMode=BMM_Additive
    translationSpace=BCS_BoneSpace
    alphaBoolBlend {blendOption=Linear}
    alphaScaleBiasClamp {bInterpResult=true}
    boneToModify {boneName=dyn_hair_bck_01}
    rotation {pitch=15}
   }
   "Transform (Modify) Bone - Bone: dyn_hair_bck_02"(Alpha=1, Translation="0,0,0", Rotation="0,0,0", Scale="1,1,1") {
    rotationMode=BMM_Additive
    rotationSpace=BCS_BoneSpace
    scaleSpace=BCS_BoneSpace
    translationMode=BMM_Additive
    translationSpace=BCS_BoneSpace
    alphaBoolBlend {blendOption=Linear}
    alphaScaleBiasClamp {bInterpResult=true}
    boneToModify {boneName=dyn_hair_bck_02}
   }
   n1 = "Transform (Modify) Bone - Bone: dyn_hair_side_r_01"(Alpha=1, Translation="1,-2,1", Rotation="0,0,0", Scale="1,1,1") {
    rotationMode=BMM_Additive
    rotationSpace=BCS_BoneSpace
    scaleSpace=BCS_BoneSpace
    translationMode=BMM_Additive
    translationSpace=BCS_BoneSpace
    alphaBoolBlend {blendOption=Linear}
    boneToModify {boneName=dyn_hair_side_r_01}
    translation=<1 -2 1>
   }
   "Transform (Modify) Bone - Bone: dyn_hair_bck_02"(ComponentPose=@n1, Alpha=1, Translation="0,0,0", Rotation="0,0,0", Scale="1,1,1") {
    rotationMode=BMM_Additive
    rotationSpace=BCS_BoneSpace
    scaleSpace=BCS_BoneSpace
    translationMode=BMM_Additive
    translationSpace=BCS_BoneSpace
    alphaBoolBlend {blendOption=Linear}
    boneToModify {boneName=dyn_hair_bck_02}
   }
   n2 = "Transform (Modify) Bone - Bone: dyn_hair_bck_01"(Alpha=1, Translation="0,0,0", Rotation="0,0,0", Scale="1,1,1.5") {
    rotationMode=BMM_Additive
    rotationSpace=BCS_BoneSpace
    scaleMode=BMM_Replace
    translationMode=BMM_Additive
    translationSpace=BCS_BoneSpace
    alphaBoolBlend {blendOption=Linear}
    boneToModify {boneName=dyn_hair_bck_01}
    scale {z=1.5}
   }
   n3 = "Transform (Modify) Bone - Bone: dyn_hair_bck_02"(ComponentPose=@n2, Alpha=1, Translation="0,0,0", Rotation="0,0,0", Scale="1,1,1") {
    rotationMode=BMM_Additive
    rotationSpace=BCS_BoneSpace
    translationMode=BMM_Additive
    translationSpace=BCS_BoneSpace
    alphaBoolBlend {blendOption=Linear}
    boneToModify {boneName=dyn_hair_bck_02}
   }
   n4 = "Transform (Modify) Bone - Bone: dyn_hair_side_l_02"(ComponentPose=@n3, Alpha=1, Translation="0,0,0", Rotation="0,0,0", Scale="1,1,1") {
    rotationMode=BMM_Additive
    rotationSpace=BCS_BoneSpace
    translationMode=BMM_Additive
    translationSpace=BCS_BoneSpace
    alphaBoolBlend {blendOption=Linear}
    boneToModify {boneName=dyn_hair_side_l_02}
   }
   n5 = "Transform (Modify) Bone - Bone: dyn_hair_side_r_01"(ComponentPose=@n4, Alpha=1, Translation="0,0,0", Rotation="0,0,0", Scale="1,1,1") {
    rotationMode=BMM_Additive
    rotationSpace=BCS_BoneSpace
    translationMode=BMM_Additive
    translationSpace=BCS_BoneSpace
    alphaBoolBlend {blendOption=Linear}
    boneToModify {boneName=dyn_hair_side_r_01}
   }
   "Transform (Modify) Bone - Bone: dyn_hair_side_r_02"(ComponentPose=@n5, Alpha=1, Translation="0,0,0", Rotation="0,0,0", Scale="1,1,1") {
    rotationMode=BMM_Additive
    rotationSpace=BCS_BoneSpace
    translationMode=BMM_Additive
    translationSpace=BCS_BoneSpace
    alphaBoolBlend {blendOption=Linear}
    boneToModify {boneName=dyn_hair_side_r_02}
   }
   n6 = Control Rig(Source=Copy Pose From Mesh() {bUseAttachedParent=true}) {
    alphaBoolBlend {blendOption=Linear}
    controlRigAssetReference {blueprintRigClass=&/F_MED_Photographer_Holiday_FaceAcc_CtrlRig.F_MED_Photographer_Holiday_FaceAcc_CtrlRig_C}
   }
   n7 = Get bIsInFrontEnd()
   n8 = RigidBody!(
    ComponentPose=Local To Component(LocalPose=@n6)
    bAlphaBoolEnabled:bool=NOT Boolean(A=@n7)
    OverrideWorldGravity:Vector="0,0,-800"
    ExternalForce
   ) -> (Pose) {
    alphaInputType=Bool
    bClampLinearTranslationLimitToRefPose=true
    bOverrideWorldGravity=true
    overridePhysicsAsset=&/F_MED_Photographer_Holiday_FaceAcc_New_Physics
    simulationSpace=BaseBoneSpace
    alphaBoolBlend {blendInTime=0.1, blendOption=Linear, blendOutTime=0.1}
    baseBoneRef {boneName=spine_05}
    componentAppliedLinearAccClamp=<3000 3000 3000>
    componentLinearAccScale=<0.4 0.4 0.4>
    componentLinearVelScale=<0.2 0.2 0.2>
    overrideWorldGravity {z=-800}
    simSpaceSettings {maxAngularVelocity=1000, maxLinearVelocity=10, externalLinearDragV=<0.5 0.5 0.5>}
   }
   n9 = RigidBody(ComponentPose=@n8, bAlphaBoolEnabled=@n7, OverrideWorldGravity="0,0,-800") {
    alphaInputType=Bool
    bClampLinearTranslationLimitToRefPose=true
    bOverrideWorldGravity=true
    overridePhysicsAsset=&/F_MED_Photographer_Holiday_FaceAcc_New_FA_Physics
    simulationSpace=BaseBoneSpace
    alphaBoolBlend {blendInTime=0.1, blendOption=Linear, blendOutTime=0.1}
    baseBoneRef {boneName=spine_05}
    componentAppliedLinearAccClamp=<3000 3000 3000>
    componentLinearAccScale=<0.4 0.4 0.4>
    componentLinearVelScale=<0.2 0.2 0.2>
    overrideWorldGravity {z=-800}
    simSpaceSettings {maxAngularVelocity=1000, maxLinearVelocity=10, externalLinearDragV=<0.5 0.5 0.5>}
   }
   n10 = "Transform (Modify) Bone - Bone: dyn_headBack_01_r"(ComponentPose=@n9, Alpha=1, Translation="0,0,0", Rotation="0,0,0", Scale="1,1,1") {
    rotationMode=BMM_Additive
    rotationSpace=BCS_BoneSpace
    scaleSpace=BCS_BoneSpace
    translationMode=BMM_Additive
    translationSpace=BCS_BoneSpace
    alphaBoolBlend {blendOption=Linear}
    alphaScaleBiasClamp {bInterpResult=true}
    boneToModify {boneName=dyn_headBack_01_r}
   }
   n11 = "Transform (Modify) Bone - Bone: dyn_head_02_r"(ComponentPose=@n10, Alpha=1, Translation="0,0,0", Rotation="0,35,-10", Scale="1,1,1") {
    rotationMode=BMM_Additive
    rotationSpace=BCS_BoneSpace
    scaleSpace=BCS_BoneSpace
    translationMode=BMM_Additive
    translationSpace=BCS_BoneSpace
    alphaBoolBlend {blendOption=Linear}
    alphaScaleBiasClamp {bInterpResult=true}
    boneToModify {boneName=dyn_head_02_r}
    rotation {roll=-10, yaw=35}
   }
   n12 = "Transform (Modify) Bone - Bone: dyn_headFront_02_r"(ComponentPose=@n11, Alpha=1, Translation="-1,0,0", Rotation="-0.004,35,0", Scale="0.8,0.7,1") {
    rotationMode=BMM_Additive
    rotationSpace=BCS_BoneSpace
    scaleMode=BMM_Additive
    scaleSpace=BCS_BoneSpace
    translationMode=BMM_Additive
    translationSpace=BCS_BoneSpace
    alphaBoolBlend {blendOption=Linear}
    alphaScaleBiasClamp {bInterpResult=true}
    boneToModify {boneName=dyn_headFront_02_r}
    rotation {pitch=-0.004, yaw=35}
    scale=<0.8 0.7>
    translation {x=-1}
   }
   n13 = "Transform (Modify) Bone - Bone: dyn_headFront_03_r"(ComponentPose=@n12, Alpha=1, Translation="0,0,0", Rotation="-0.0004,35,0", Scale="1,1,1") {
    rotationMode=BMM_Additive
    rotationSpace=BCS_BoneSpace
    scaleMode=BMM_Additive
    scaleSpace=BCS_BoneSpace
    translationMode=BMM_Additive
    translationSpace=BCS_BoneSpace
    alphaBoolBlend {blendOption=Linear}
    alphaScaleBiasClamp {bInterpResult=true}
    boneToModify {boneName=dyn_headFront_03_r}
    rotation {pitch=-0.0004, yaw=35}
   }
   n14 = Anim Dynamics - Bone: dyn_pompom_01!(ComponentPose=@n13, Alpha=1, GravityOverride:Vector="0,0,-350", ExternalForce="0,0,0") -> (Pose) {
    angularDampingOverride=0.8
    angularSpringConstant=75
    bAngularSpring=true
    bGravityOverrideInSimSpace=true
    bUseGravityOverride=true
    linearDampingOverride=0.9
    linearSpringConstant=150
    numSolverIterationsPreUpdate=3
    alphaBoolBlend {blendOption=Linear}
    boundBone {boneName=dyn_pompom_01}
    componentAppliedLinearAccClamp=<10000 10000 10000>
    componentLinearAccScale=<0.1 0.1 0.2>
    componentLinearVelScale=<0.1 0.1 0.1>
    gravityOverride {z=-350}
    physicsBodyDefinitions [{
      boundBone {boneName=dyn_pompom_01}
      boxExtents=<9 9 9>
      constraintSetup {angularLimitsMax=<10 15 8>, angularLimitsMin=<0 -60 -8>, angularTarget=<1 0 0>}
      localJointOffset=<2 0 0>
     }]
    relativeSpaceBone {boneName=head}
   }
   n15 = Anim Dynamics - Bone: dyn_pompom_02!(ComponentPose=@n14, Alpha=1, GravityOverride:Vector="0,0,-200", ExternalForce="0,0,0") -> (Pose) {
    angularDampingOverride=0.8
    angularSpringConstant=75
    bAngularSpring=true
    bGravityOverrideInSimSpace=true
    bUseGravityOverride=true
    linearDampingOverride=0.9
    linearSpringConstant=150
    numSolverIterationsPreUpdate=3
    alphaBoolBlend {blendOption=Linear}
    boundBone {boneName=dyn_pompom_02}
    componentAppliedLinearAccClamp=<10000 10000 10000>
    componentLinearAccScale=<0.1 0.1 0.1>
    componentLinearVelScale=<0.1 0.1 0.1>
    gravityOverride {z=-200}
    physicsBodyDefinitions [{
      boundBone {boneName=dyn_pompom_02}
      boxExtents=<15 15 15>
      constraintSetup {angularLimitsMax=<10 15 0>, angularLimitsMin=<0 -60 0>, angularTarget=<1 0 0>}
      localJointOffset=<1 0 0>
     }]
    relativeSpaceBone {boneName=head}
   }
   n16 = "Constraint - dyn_pompom_02 (PP)"!(ComponentPose=@n15, Alpha=1, ConstraintWeights_0=0, ConstraintWeights_1:float=1) -> (Pose) {
    alphaBoolBlend {blendOption=Linear}
    boneToModify {boneName=dyn_pompom_02}
    constraintSetup [{targetBone {boneName=dyn_hatHair_MSTR}}, {targetBone {boneName=dyn_pompom_01}}]
    constraintWeights [1,1]
   }
   n17 = Anim Dynamics - Bone: dyn_ear_r(ComponentPose=@n16, Alpha=1, ExternalForce="0,0,0") {
    angularSpringConstant=0.3
    bGravityOverrideInSimSpace=true
    bUseGravityOverride=true
    alphaBoolBlend {blendOption=Linear}
    boundBone {boneName=dyn_ear_r}
    componentAppliedLinearAccClamp=<3000 3000 3000>
    componentLinearAccScale=<0.8 0.8 0.8>
    componentLinearVelScale=<0.2 0.2 0.2>
    gravityOverride {z=-980}
    physicsBodyDefinitions [{
      boundBone {boneName=dyn_ear_r}
      boxExtents=<10 10 10>
      constraintSetup {angularLimitsMax=<20 20 20>, angularLimitsMin=<-10 -10 -10>, angularTarget=<0 0 0>}
      localJointOffset=<0 0 0>
     }]
    relativeSpaceBone {boneName=head}
   }
   n18 = Anim Dynamics - Bone: dyn_ear_l(ComponentPose=@n17, Alpha=1, ExternalForce="0,0,0") {
    angularSpringConstant=0.3
    bGravityOverrideInSimSpace=true
    bUseGravityOverride=true
    alphaBoolBlend {blendOption=Linear}
    boundBone {boneName=dyn_ear_l}
    componentAppliedLinearAccClamp=<3000 3000 3000>
    componentLinearAccScale=<0.8 0.8 0.8>
    componentLinearVelScale=<0.2 0.2 0.2>
    gravityOverride {z=-980}
    physicsBodyDefinitions [{
      boundBone {boneName=dyn_ear_l}
      boxExtents=<10 10 10>
      constraintSetup {angularLimitsMax=<20 20 20>, angularLimitsMin=<-10 -10 -10>, angularTarget=<0 0 0>}
      localJointOffset=<0 0 0>
     }]
    relativeSpaceBone {boneName=head}
   }
   n19 = "Transform (Modify) Bone - Bone: dyn_hair_bck_01"(ComponentPose=@n18, Alpha=1, Translation="1,1,3", Rotation="0,0,0", Scale="1,1,1") {
    rotationMode=BMM_Additive
    translationMode=BMM_Additive
    alphaBoolBlend {blendOption=Linear}
    alphaScaleBiasClamp {bInterpResult=true}
    boneToModify {boneName=dyn_hair_bck_01}
    translation=<1 1 3>
   }
   n20 = "Transform (Modify) Bone - Bone: dyn_hair_side_l_01"(ComponentPose=@n19, Alpha=1, Translation="0,0,2", Rotation="0,0,0", Scale="1,1,1") {
    rotationMode=BMM_Additive
    translationMode=BMM_Additive
    alphaBoolBlend {blendOption=Linear}
    alphaScaleBiasClamp {bInterpResult=true}
    boneToModify {boneName=dyn_hair_side_l_01}
    translation {z=2}
   }
   n21 = "Transform (Modify) Bone - Bone: dyn_pompom_01"(ComponentPose=@n20, Alpha=1, Translation="0,0,1.5", Rotation="0,0,65", Scale="1,1,1") {
    rotationMode=BMM_Additive
    translationMode=BMM_Additive
    alphaBoolBlend {blendOption=Linear}
    alphaScaleBiasClamp {bInterpResult=true}
    boneToModify {boneName=dyn_pompom_01}
    rotation {roll=65}
    translation {z=1.5}
   }
   n22 = "Transform (Modify) Bone - Bone: dyn_pompom_02"(ComponentPose=@n21, Alpha=1, Translation="0,2,-2", Rotation="0,0,0", Scale="1,1,1") {
    rotationMode=BMM_Additive
    translationMode=BMM_Additive
    alphaBoolBlend {blendOption=Linear}
    alphaScaleBiasClamp {bInterpResult=true}
    boneToModify {boneName=dyn_pompom_02}
    translation {y=2, z=-2}
   }
   n23 = "Transform (Modify) Bone - Bone: dyn_hair_bck_01"(ComponentPose=@n22, Alpha=1, Translation="0,1,0", Rotation="0,0,0", Scale="1,1,1") {
    rotationMode=BMM_Additive
    translationMode=BMM_Additive
    alphaBoolBlend {blendOption=Linear}
    alphaScaleBiasClamp {bInterpResult=true}
    boneToModify {boneName=dyn_hair_bck_01}
    translation {y=1}
   }
   n24 = "Transform (Modify) Bone - Bone: dyn_hair_side_l_01"(ComponentPose=@n23, Alpha=1, Translation="0,1,0", Rotation="0,0,0", Scale="1,1,1") {
    rotationMode=BMM_Additive
    translationMode=BMM_Additive
    alphaBoolBlend {blendOption=Linear}
    alphaScaleBiasClamp {bInterpResult=true}
    boneToModify {boneName=dyn_hair_side_l_01}
    translation {y=1}
   }
   n25 = "Transform (Modify) Bone - Bone: dyn_hair_side_r_01"(ComponentPose=@n24, Alpha=1, Translation="0,1,0", Rotation="0,0,0", Scale="1,1,1") {
    rotationMode=BMM_Additive
    translationMode=BMM_Additive
    alphaBoolBlend {blendOption=Linear}
    alphaScaleBiasClamp {bInterpResult=true}
    boneToModify {boneName=dyn_hair_side_r_01}
    translation {y=1}
   }
   n26 = "Transform (Modify) Bone - Bone: dyn_pompom_01"(ComponentPose=@n25, Alpha=1, Translation="0,0,0", Rotation="35,0,0", Scale="1,1,1") {
    rotationMode=BMM_Additive
    rotationSpace=BCS_BoneSpace
    alphaBoolBlend {blendOption=Linear}
    alphaScaleBiasClamp {bInterpResult=true}
    boneToModify {boneName=dyn_pompom_01}
    rotation {pitch=35}
   }
   Output Pose(Result=Component To Local(ComponentPose=@n26)) {layerGroup=None}
  ]}
 EventGraph {nodes [
   Event BlueprintUpdateAnimation()
   > n1 = Sequence!(execute) -> (then_0, then_1, then_2:Exec)
   > n2 = Branch(Condition=OR Boolean(A=Get bIsSkydiveFloating(), B=Get bIsSkydiveDiveMode()))
   n3 = Cast To FortPlayerAnimInstance(execute=@n2.else, Object=GetAnimInstance(self=GetOwnerSkeletalMeshComponent()))
   > Set lookup targeting(lookup targeting:double=0) -> (Output_Get:double)
   > n4 = Switch on EFortWeaponCoreAnimation!(execute, Selection:EFortWeaponCoreAnimation Enum=Get WeaponCoreAnim(self=@n3)) -> (
    Melee:Exec
    Pistol:Exec
    Shotgun:Exec
    PaperBlueprint:Exec
    Rifle:Exec
    MeleeOneHand:Exec
    MachinePistol:Exec
    RocketLauncher:Exec
    GrenadeLauncher:Exec
    GoingCommando:Exec
    AssaultRifle:Exec
    TacticalShotgun:Exec
    SniperRifle:Exec
    TrapPlacement:Exec
    ShoulderLauncher:Exec
    AbilityDecoTool:Exec
    Crossbow:Exec
    C4:Exec
    RemoteControl:Exec
    DualWield:Exec
    AR_BullPup:Exec
    AR_ForwardGrip:Exec
    MedPackPaddles:Exec
    SMG_P90:Exec
    AR_DrumGun:Exec
    Consumable_Small:Exec
    Consumable_Large:Exec
    Balloon:Exec
    MountedTurret:Exec
    CreativeTool:Exec
    ExplosiveBow:Exec
    AshtonIndigo:Exec
    AshtonChicago:Exec
    MeleeDualWield:Exec
    Unarmed:Exec
   )
   n5 = BreakTransform(InTransform=GetDeltaTransformFromRefPose(self=GetOwnerSkeletalMeshComponent(), BoneName=head, BaseName=neck_02))
   n6 = Set headRot alpha(
    execute=[@n4.DualWield, @n4.Pistol]
    headRot alpha:double=MapRangeClamped(Value=BreakRotator(InRot=@n5.Rotation).Yaw, InRangeA=-5, InRangeB=-20)
   ) -> (Output_Get:double)
   n7 = float * float(A=@n6, B="To Float (Boolean)"(InBool=Get bIsTargeting())) {is=title K2Node_PromotableOperator}
   > Set lookup targeting(lookup targeting=@n7)
   > n8 = Branch(Condition="float > float"(A=@n7))
   > n9 = Set rban alpha(rban alpha:double=0.8) -> (Output_Get:double)
   n10 = GetOwnerSkeletalMeshComponent()
   n11 = BreakRotator(InRot=BreakTransform(InTransform=GetDeltaTransformFromRefPose(self=@n10, BoneName=upperarm_r, BaseName=spine_05)).Rotation)
   Set Arm Up R(execute=[@n9.then, @n18.then, @n14.then], Arm Up R:double=MapRangeClamped(Value="Absolute (Float)"(A=@n11.Yaw), InRangeA=0, InRangeB=100)) -> (Output_Get:double)
   n12 = GetOwnerSkeletalMeshComponent()
   n13 = BreakRotator(InRot=BreakTransform(InTransform=GetDeltaTransformFromRefPose(self=@n12, BoneName=upperarm_l, BaseName=spine_05)).Rotation)
   > Set Arm Up L(Arm Up L:double=MapRangeClamped(Value="Absolute (Float)"(A=@n13.Yaw), InRangeA=40, InRangeB=100)) -> (Output_Get:double)
   n14 = Set rban alpha(execute=@n8.else, rban alpha=1)
   n15 = Branch(execute=@n1.then_1, Condition=Get bIsCrouching())
   > n16 = Branch(Condition=Get bIsMoving2D())
   > Set Crouch Move Alpha(Crouch Move Alpha:double=1) -> (Output_Get:double)
   Set Crouch Move Alpha(execute=@n16.else, Crouch Move Alpha=0)
   Set Crouch Move Alpha(execute=@n15.else, Crouch Move Alpha=0)
   n17 = Gravity Override State Function DEPRECATED(
    execute=@n1.then_2
    CharacterPartAnimInstance=Self-Reference()
    GameTimeInSeconds=GetGameTimeInSeconds()
    GravityOverrideParams=Get Gravity Override Params()
   )
   > Set Gravity Override(Gravity Override=@n17)
   n18 = Set headRot alpha(headRot alpha=0)
  ]}
]
variables [
 Arm Up L {type=number}
 Arm Up R {type=number}
 Crouch Move Alpha {type=number}
 Gravity Override {type=Vector}
 Gravity Override Params {type=GravityOverrideParamsStruct}
 headRot alpha {type=number}
 lookup targeting {type=number}
 rban alpha {type=number}
]
